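; GrillMoon countdown mask timer Blueprint that forces the correct Niagara tether VFX on for clients during the event.
; On BeginPlay, only the non-authority side starts a looping four-second timer that reads Mask_Phase from the GrillMoon Material Parameter Collection.
; Each phase has a DoOnce-guarded branch: the first activates the first mask effect, later ones activate PortalContact.
; All but the final phase also deactivate a numbered mask effect.
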(event EventBeginPlay
  ; Force the correct VFX to turn on
  (Networking|SwitchHasAuthority
    (:Remote
      (Utilities|Time|SetTimerbyEvent (delegate Custom|Check%20and%20Activate%20VFX%20Tethers) 4.0 true))))

(event Custom|Check%20and%20Activate%20VFX%20Tethers
  ; Force the correct VFX to turn on
  (bind _returnvalue (Rendering|Material|GetScalarParameterValue (Variables|Default|GetMPC_GrillMoon) (Utilities|Name|MakeLiteralName "Mask_Phase") :!class /Script/Engine.KismetMaterialLibrary))
  (if (Math|Float|NearlyEqual(Float) _returnvalue 1.0 0.01)
    (Utilities|FlowControl|DoOnce false
      (:Completed
        (Components|Activation|Activate (Variables|Default|GetNS_GrillMoon_Countdown_Masks_001) false)
        (Components|Activation|Deactivate (Variables|Default|GetNS_GrillMoon_Countdown_Masks_002))))
    (elif (Math|Float|NearlyEqual(Float) _returnvalue 2.0 0.01)
      (Utilities|FlowControl|DoOnce false
        (:Completed
          (Components|Activation|Activate (Variables|Default|GetNS_GrillMoon_Countdown_PortalContact) false)
          (Components|Activation|Deactivate (Variables|Default|GetNS_GrillMoon_Countdown_Masks_003))))
      (elif (Math|Float|NearlyEqual(Float) _returnvalue 3.0 0.01)
        (Utilities|FlowControl|DoOnce false
          (:Completed
            (Components|Activation|Activate (Variables|Default|GetNS_GrillMoon_Countdown_PortalContact) false)
            (Components|Activation|Deactivate (Variables|Default|GetNS_GrillMoon_Countdown_Masks_004))))
        (elif (Math|Float|NearlyEqual(Float) _returnvalue 4.0 0.01)
          (Utilities|FlowControl|DoOnce false
            (:Completed
              (Components|Activation|Activate (Variables|Default|GetNS_GrillMoon_Countdown_PortalContact) false)
              (Components|Activation|Deactivate (Variables|Default|GetNS_GrillMoon_Countdown_Masks_005))))
          (elif (Math|Float|NearlyEqual(Float) _returnvalue 5.0 0.01)
            (Utilities|FlowControl|DoOnce false
              (:Completed
                (Components|Activation|Activate (Variables|Default|GetNS_GrillMoon_Countdown_PortalContact) false)))))))))

(fn ConstructionScript)
